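# Reticle widget for the HeatSnake Drift dual weapons, derived from the dual-weapon reticle base and showing a four-prong crosshair image.
# On initialization it caches the image's dynamic material instance.
# Each tick, when both primary and offhand weapons are valid, it sets the material's SpreadOffset to the larger of their spread-to-field-of-view ratios, so the crosshair follows the wider spread.
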WidgetBlueprint /HeatSnake_Drift/UI/WBP_Reticle_HeatSnakeDrift
graphs [EventGraph {nodes [
   Event Tick()
   n1 = Get varOffhandWeapon() -> (varOffhandWeapon:Fort Weapon Ranged Object Reference)
   n2 = Get varPrimaryWeapon() -> (varPrimaryWeapon:Fort Weapon Ranged Object Reference)
   > Branch(Condition=AND Boolean(A=IsValid(Object=@n2), B=IsValid(Object=@n1)))
   n3 = Get owningFortPlayerController() -> (owningFortPlayerController:Fort Player Controller Object Reference)
   n4 = Get varOffhandWeapon()
   n5 = Get varPrimaryWeapon()
   > SetScalarParameterValue(
    self=Get CrosshairDynamicMaterial() -> (CrosshairDynamicMaterial:Material Instance Dynamic Object Reference)
    ParameterName=SpreadOffset
    Value="Max (Float)"(
     A=GetSpreadToFieldOfViewRatio!(self:Fort Weapon Ranged Object Reference=@n5, PlayerController:Fort Player Controller Object Reference=@n3, bUseMinimumSpread:bool=false) -> (ReturnValue:float)
     B=GetSpreadToFieldOfViewRatio(self=@n4, PlayerController=@n3, bUseMinimumSpread=false)
    )
   )
   Event OnInitialized()
   > Parent: OnInitialized() {is=title K2Node_CallParentFunction}
   > n6 = GetDynamicMaterial(self=Get Image_Crosshair() -> (Image_Crosshair:Image Object Reference))
   > Set CrosshairDynamicMaterial(CrosshairDynamicMaterial:Material Instance Dynamic Object Reference=@n6) -> (Output_Get:Material Instance Dynamic Object Reference)
  ]}]
variables [CrosshairDynamicMaterial {type=MaterialInstanceDynamic}]
widgets {
 info {
  inheritedWidgetCount=0
  namedSlotCount=1
  parentClass=/BattleRoyaleDualWeaponUI/Widgets/WBP_Reticle_DualWeaponBase.WBP_Reticle_DualWeaponBase_C
  rootWidgetClass=null
  widgetCount=1
 }
 widgets [Image_Crosshair {
   widgetClassPath=&Image
   bIsVariable=true
   widget=Image_Crosshair
   properties {
    visibility=SelfHitTestInvisible
    brush {
     bIsDynamicallyLoaded=false
     drawAs=Image
     imageType=FullColor
     mirroring=NoMirror
     resourceName=None
     resourceObject=&/MI_4Prong_Cross_HeatSnakeDrift
     tiling=NoTile
     imageSize=<128 128>
     margin=<ltrb 0 0 0 0>
     outlineSettings {
      bUseBrushTransparency=false
      roundingType=HalfHeightRadius
      width=0
      color {colorUseRule=UseColor_Specified, specifiedColor=<rgba 0 0 0 0>}
      cornerRadii=<0 0 0 0>
     }
     tintColor {colorUseRule=UseColor_Specified, specifiedColor=<rgba 1 1 1 1>}
     uVRegion {bIsValid=false, max=<0 0>, min=<0 0>}
    }
   }
  }]
}
